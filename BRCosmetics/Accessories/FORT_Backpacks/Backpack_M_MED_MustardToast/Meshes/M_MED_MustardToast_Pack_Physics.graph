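# Physics asset for the MustardToast medium backpack cosmetic, defining capsule bodies and joints for its straps, side loops and zipper.
# A large capsule covers the pack, with upper and lower strap bases joined to it by free-rotation constraints.
# Each strap chains dynamic segments that bend mainly on one swing axis, with zero twist limits.
# Loops swing on the other axis and the zipper on both, each jointed only to its own base.
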PhysicsAsset /BRCosmetics/Accessories/FORT_Backpacks/Backpack_M_MED_MustardToast/Meshes/M_MED_MustardToast_Pack_Physics
bodies {
 backpack_base [{shapeType=Capsule, length=13, radius=8.7, shapeName=backpack_base_capsule, center=<7.5 -0.5 0>, rotation=<rot 0 0 180>}]
 loop_base_l [{shapeType=Capsule, length=1, radius=1, shapeName=loop_base_l_capsule, center=<1 0 0>, rotation=<rot 90 0 0>}]
 loop_base_r [{shapeType=Capsule, length=1, radius=1, shapeName=loop_base_r_capsule, center=<1 0 0>, rotation=<rot 90 0 0>}]
 loop_dyn_l [{shapeType=Capsule, length=1.8, radius=1, shapeName=loop_dyn_l_capsule, center=<1 0 0>, rotation=<rot 90 0 0>}]
 loop_dyn_r [{shapeType=Capsule, length=1.8, radius=1, shapeName=loop_dyn_r_capsule, center=<1 0 0>, rotation=<rot 90 0 0>}]
 strap_lower_01_dyn [{shapeType=Capsule, length=1.3, radius=1, shapeName=strap_lower_01_dyn_capsule, center=<2 0 0>, rotation=<rot 90 0 0>}]
 strap_lower_02_dyn [{shapeType=Capsule, length=1.3, radius=1, shapeName=strap_lower_02_dyn_capsule, center=<1.5 0 0>, rotation=<rot 90 0 0>}]
 strap_lower_03_dyn [{shapeType=Capsule, length=1.3, radius=1, shapeName=strap_lower_03_dyn_capsule, center=<1.5 0 0>, rotation=<rot 90 0 0>}]
 strap_lower_04_dyn [{shapeType=Capsule, length=1.3, radius=1, shapeName=strap_lower_04_dyn_capsule, center=<1.5 0 0>, rotation=<rot 90 0 0>}]
 strap_lower_base [{shapeType=Capsule, length=1, radius=1, shapeName=strap_lower_base_capsule, center=<1 0 0>, rotation=<rot 90 0 0>}]
 strap_upper_01_dyn [{shapeType=Capsule, length=1.8, radius=1, shapeName=strap_upper_01_dyn_capsule, center=<1.1 0 0>, rotation=<rot 90 0 0>}]
 strap_upper_02_dyn [{shapeType=Capsule, length=1.8, radius=1.1, shapeName=strap_upper_02_dyn_capsule, center=<1.15 0 0>, rotation=<rot 90 0 0>}]
 strap_upper_03_dyn [{shapeType=Capsule, length=1.8, radius=1.1, shapeName=strap_upper_03_dyn_capsule, center=<1.2 0 0>, rotation=<rot 90 0 0>}]
 strap_upper_base [{shapeType=Capsule, length=1, radius=1, shapeName=strap_upper_base_capsule, center=<1 0 0>, rotation=<rot 90 0 0>}]
 zipper_base [{shapeType=Capsule, length=1, radius=1, shapeName=zipper_base_capsule, center=<1 0 0>, rotation=<rot 90 0 0>}]
 zipper_dyn [{shapeType=Capsule, length=1.8, radius=1, shapeName=zipper_dyn_capsule, center=<1 0 0>, rotation=<rot 90 0 0>}]
}
constraints [
 {
  bone1Name=strap_lower_base
  bone2Name=backpack_base
  swing1LimitDegrees=45
  swing1Motion=Free
  swing2LimitDegrees=45
  swing2Motion=Free
  twistLimitDegrees=45
  twistMotion=Free
 }
 {bone1Name=strap_lower_01_dyn, bone2Name=strap_lower_base, swing1LimitDegrees=45, swing2LimitDegrees=0, twistLimitDegrees=0}
 {bone1Name=strap_lower_02_dyn, bone2Name=strap_lower_01_dyn, swing1LimitDegrees=35, swing2LimitDegrees=5, twistLimitDegrees=0}
 {bone1Name=strap_lower_03_dyn, bone2Name=strap_lower_02_dyn, swing1LimitDegrees=45, swing2LimitDegrees=5, twistLimitDegrees=0}
 {bone1Name=strap_lower_04_dyn, bone2Name=strap_lower_03_dyn, swing1LimitDegrees=55, swing2LimitDegrees=5, twistLimitDegrees=0}
 {
  bone1Name=strap_upper_base
  bone2Name=backpack_base
  swing1LimitDegrees=45
  swing1Motion=Free
  swing2LimitDegrees=45
  swing2Motion=Free
  twistLimitDegrees=45
  twistMotion=Free
 }
 {bone1Name=strap_upper_01_dyn, bone2Name=strap_upper_base, swing1LimitDegrees=25, swing2LimitDegrees=0, twistLimitDegrees=0}
 {bone1Name=strap_upper_02_dyn, bone2Name=strap_upper_01_dyn, swing1LimitDegrees=30, swing2LimitDegrees=0, twistLimitDegrees=0}
 {bone1Name=strap_upper_03_dyn, bone2Name=strap_upper_02_dyn, swing1LimitDegrees=35, swing2LimitDegrees=2, twistLimitDegrees=0}
 {bone1Name=loop_dyn_l, bone2Name=loop_base_l, swing1LimitDegrees=0, swing2LimitDegrees=80, twistLimitDegrees=0}
 {bone1Name=loop_dyn_r, bone2Name=loop_base_r, swing1LimitDegrees=0, swing2LimitDegrees=120, twistLimitDegrees=0}
 {bone1Name=zipper_dyn, bone2Name=zipper_base, swing1LimitDegrees=45, swing2LimitDegrees=45, twistLimitDegrees=0}
]
