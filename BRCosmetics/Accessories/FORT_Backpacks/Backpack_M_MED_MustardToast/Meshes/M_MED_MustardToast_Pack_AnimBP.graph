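# Post-input AnimBP for the MustardToast backpack cosmetic, adding secondary motion to the pack, boxes, zippers and straps.
# Strap bases rotate additively while skydive floating, then chained spring AnimDynamics nodes feed a cached pose.
# A base-bone-space RigidBody pass simulates that cache, and rotation constraints make strap loop bones follow their straps.
# The event graph updates GravityOverride through a deprecated function, but the anim nodes use fixed gravity values.
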AnimBlueprint /BRCosmetics/Accessories/FORT_Backpacks/Backpack_M_MED_MustardToast/Meshes/M_MED_MustardToast_Pack_AnimBP
target_skeleton=&/M_MED_MustardToast_Pack_Skeleton
graphs [
 AnimGraph {nodes [
   n1 = RigidBody!(
    ComponentPose=Local To Component(LocalPose=Use cached pose 'AnimDynamics Cache'())
    Alpha
    OverrideWorldGravity:Vector="0,0,0"
    ExternalForce
   ) -> (Pose) {
    bOverrideWorldGravity=true
    overridePhysicsAsset=&/M_MED_MustardToast_Pack_Physics
    simulationSpace=BaseBoneSpace
    alphaBoolBlend {blendOption=Linear}
    baseBoneRef {boneName=root}
    componentAppliedLinearAccClamp=<3000 3000 3000>
    componentLinearAccScale=<0.3 0.3 0.3>
    componentLinearVelScale=<0.3 0.3 0.3>
    simSpaceSettings {
     maxAngularAcceleration=1000
     maxAngularVelocity=3
     maxLinearAcceleration=700
     maxLinearVelocity=500
     velocityScaleZ=0
     worldAlpha=0.4
     externalLinearDragV=<0.2 0.2>
    }
   }
   n2 = "Constraint - strap_lower_loop_dyn_A (R)"(ComponentPose=@n1, Alpha=1, ConstraintWeights_0=1) {
    alphaBoolBlend {blendOption=Linear}
    boneToModify {boneName=strap_lower_loop_dyn_A}
    constraintSetup [{transformType=Rotation, targetBone {boneName=strap_lower_01_dyn}}]
    constraintWeights [1]
   }
   n3 = "Constraint - strap_lower_loop_dyn_B (R)"(ComponentPose=@n2, Alpha=1, ConstraintWeights_0=1) {
    alphaBoolBlend {blendOption=Linear}
    boneToModify {boneName=strap_lower_loop_dyn_B}
    constraintSetup [{transformType=Rotation, targetBone {boneName=strap_lower_01_dyn}}]
    constraintWeights [1]
   }
   n4 = "Constraint - strap_upper_loop_dyn_A (R)"(ComponentPose=@n3, Alpha=1, ConstraintWeights_0=1) {
    alphaBoolBlend {blendOption=Linear}
    boneToModify {boneName=strap_upper_loop_dyn_A}
    constraintSetup [{transformType=Rotation, targetBone {boneName=strap_upper_01_dyn}}]
    constraintWeights [1]
   }
   n5 = "Constraint - strap_upper_loop_dyn_B (R)"(ComponentPose=@n4, Alpha=1, ConstraintWeights_0=1) {
    alphaBoolBlend {blendOption=Linear}
    boneToModify {boneName=strap_upper_loop_dyn_B}
    constraintSetup [{transformType=Rotation, targetBone {boneName=strap_upper_01_dyn}}]
    constraintWeights [1]
   }
   Output Pose(Result=Component To Local(ComponentPose=@n5)) {layerGroup=None}
   n6 = Get bIsSkydiveFloating()
   n7 = "Transform (Modify) Bone - Bone: strap_upper_base"!(ComponentPose=Local To Component(LocalPose=Input Pose()), bAlphaBoolEnabled:bool=@n6, Translation="0,0,0", Rotation="0,70,0", Scale="1,1,1") -> (Pose) {
    alphaInputType=Bool
    rotationMode=BMM_Additive
    rotationSpace=BCS_BoneSpace
    alphaBoolBlend {blendInTime=2, blendOption=Linear, blendOutTime=2}
    alphaScaleBiasClamp {bInterpResult=true}
    boneToModify {boneName=strap_upper_base}
    rotation {yaw=70}
   }
   n8 = "Transform (Modify) Bone - Bone: strap_lower_base"!(ComponentPose=@n7, bAlphaBoolEnabled:bool=@n6, Translation="0,0,0", Rotation="0,-40,0", Scale="1,1,1") -> (Pose) {
    alphaInputType=Bool
    rotationMode=BMM_Additive
    rotationSpace=BCS_BoneSpace
    alphaBoolBlend {blendInTime=2, blendOption=Linear, blendOutTime=2}
    alphaScaleBiasClamp {bInterpResult=true}
    boneToModify {boneName=strap_lower_base}
    rotation {yaw=-40}
   }
   n9 = Anim Dynamics - Bone: backpack_dyn!(ComponentPose=@n8, Alpha=0.5, GravityOverride:Vector="0,0,-980", ExternalForce="0,0,0") -> (Pose) {
    angularDampingOverride=1
    angularSpringConstant=15000
    bAngularSpring=true
    bOverrideAngularDamping=true
    bUseGravityOverride=true
    gravityScale=0.35
    alphaBoolBlend {blendOption=Linear}
    boundBone {boneName=backpack_dyn}
    componentAppliedLinearAccClamp=<3000 3000 3000>
    componentLinearAccScale=<0.2 0.2 0.2>
    componentLinearVelScale=<0.2 0.2 0.2>
    gravityOverride {z=-980}
    physicsBodyDefinitions [{
      boundBone {boneName=backpack_dyn}
      boxExtents=<10 10 10>
      constraintSetup {angularLimitsMax=<5 5 15>, angularLimitsMin=<-2 -2 -3>, angularTarget=<1 0 0>}
      localJointOffset=<1 0 0>
     }]
   }
   n10 = Anim Dynamics - Bone: Box_01_dyn!(ComponentPose=@n9, Alpha=1, GravityOverride:Vector="0,0,-980", ExternalForce="0,0,0") -> (Pose) {
    angularDampingOverride=1
    angularSpringConstant=24000
    bAngularSpring=true
    bOverrideAngularDamping=true
    bUseGravityOverride=true
    gravityScale=0.35
    alphaBoolBlend {blendOption=Linear}
    boundBone {boneName=Box_01_dyn}
    componentAppliedLinearAccClamp=<3000 3000 3000>
    componentLinearAccScale=<0.2 0.2 0.2>
    componentLinearVelScale=<0.2 0.2 0.2>
    gravityOverride {z=-980}
    physicsBodyDefinitions [{
      boundBone {boneName=Box_01_dyn}
      boxExtents=<50 50 50>
      constraintSetup {
       angularLimitsMax=<5 5 5>
       angularLimitsMin=<-5 -5 -5>
       angularTarget=<1 0 0>
       linearAxesMax=<0.25 0.25 0.25>
       linearAxesMin=<-0.25 -0.25 -0.25>
      }
      localJointOffset=<4 0 0>
     }]
   }
   n11 = Anim Dynamics - Bone: Box_02_dyn!(ComponentPose=@n10, Alpha=1, GravityOverride:Vector="0,0,-980", ExternalForce="0,0,0") -> (Pose) {
    angularDampingOverride=1
    angularSpringConstant=24000
    bAngularSpring=true
    bOverrideAngularDamping=true
    bOverrideLinearDamping=true
    bUseGravityOverride=true
    gravityScale=0.35
    linearDampingOverride=0.1
    alphaBoolBlend {blendOption=Linear}
    boundBone {boneName=Box_02_dyn}
    componentAppliedLinearAccClamp=<3000 3000 3000>
    componentLinearAccScale=<0.2 0.2 0.2>
    componentLinearVelScale=<0.2 0.2 0.2>
    gravityOverride {z=-980}
    physicsBodyDefinitions [{
      boundBone {boneName=Box_02_dyn}
      boxExtents=<50 50 50>
      constraintSetup {angularLimitsMax=<5 5 5>, angularLimitsMin=<-5 -5 -5>, angularTarget=<1 0 0>, linearAxesMax=<0.3 0.3 0.5>, linearAxesMin=<-0.3 -0.3 -0.4>}
      localJointOffset=<4 0 0>
     }]
   }
   n12 = Anim Dynamics - Bone: Box_03_dyn!(ComponentPose=@n11, Alpha=1, GravityOverride:Vector="0,0,-980", ExternalForce="0,0,0") -> (Pose) {
    angularDampingOverride=1
    angularSpringConstant=12000
    bAngularSpring=true
    bOverrideAngularDamping=true
    bUseGravityOverride=true
    gravityScale=0.35
    alphaBoolBlend {blendOption=Linear}
    boundBone {boneName=Box_03_dyn}
    componentAppliedLinearAccClamp=<3000 3000 3000>
    componentLinearAccScale=<0.2 0.2 0.2>
    componentLinearVelScale=<0.2 0.2 0.2>
    gravityOverride {z=-980}
    physicsBodyDefinitions [{
      boundBone {boneName=Box_03_dyn}
      boxExtents=<50 50 50>
      constraintSetup {angularLimitsMax=<5 5 5>, angularLimitsMin=<-5 -5 -5>, angularTarget=<1 0 0>, linearAxesMax=<0.2 0.2 0.2>, linearAxesMin=<-0.2 -0.2 -0.2>}
      localJointOffset=<2 0 0>
     }]
   }
   n13 = Anim Dynamics - Bone: strap_low_zipper_dyn!(ComponentPose=@n12, Alpha=1, GravityOverride:Vector="0,0,-980", ExternalForce="0,0,0") -> (Pose) {
    angularDampingOverride=1
    angularSpringConstant=15000
    bAngularSpring=true
    bOverrideAngularDamping=true
    bUseGravityOverride=true
    gravityScale=0.35
    alphaBoolBlend {blendOption=Linear}
    boundBone {boneName=strap_low_zipper_dyn}
    componentAppliedLinearAccClamp=<3000 3000 3000>
    componentLinearAccScale=<0.3 0.3 0.3>
    componentLinearVelScale=<0.3 0.3 0.3>
    gravityOverride {z=-980}
    physicsBodyDefinitions [{
      boundBone {boneName=strap_low_zipper_dyn}
      boxExtents=<10 10 10>
      constraintSetup {angularLimitsMax=<0 5 25>, angularLimitsMin=<0 -5 -25>, angularTarget=<1 0 0>}
      localJointOffset=<4 0 0>
     }]
   }
   n14 = Anim Dynamics - Bone: strap_up_zipper_dyn!(ComponentPose=@n13, Alpha=1, GravityOverride:Vector="0,0,-980", ExternalForce="0,0,0") -> (Pose) {
    angularDampingOverride=1
    angularSpringConstant=10000
    bAngularSpring=true
    bOverrideAngularDamping=true
    bUseGravityOverride=true
    gravityScale=0.35
    alphaBoolBlend {blendOption=Linear}
    boundBone {boneName=strap_up_zipper_dyn}
    componentAppliedLinearAccClamp=<3000 3000 3000>
    componentLinearAccScale=<0.3 0.3 0.3>
    componentLinearVelScale=<0.3 0.3 0.3>
    gravityOverride {z=-980}
    physicsBodyDefinitions [{
      boundBone {boneName=strap_up_zipper_dyn}
      boxExtents=<10 10 10>
      constraintSetup {angularLimitsMax=<0 5 25>, angularLimitsMin=<0 -5 -25>, angularTarget=<1 0 0>}
      localJointOffset=<4 0 0>
     }]
   }
   Save cached pose 'AnimDynamics Cache'(Pose=Component To Local(ComponentPose=@n14)) {cacheName=AnimDynamics Cache}
  ]}
 EventGraph {nodes [
   Event BlueprintUpdateAnimation()
   > n1 = Gravity Override State Function DEPRECATED(CharacterPartAnimInstance=Self-Reference(), GameTimeInSeconds=GetGameTimeInSeconds(), GravityOverrideParams=Get Gravity Override Params())
   > Set GravityOverride(GravityOverride=@n1)
   TryGetPawnOwner()
  ]}
]
variables [
 Gravity Override Params {type=GravityOverrideParamsStruct}
 GravityOverride {type=Vector}
]
